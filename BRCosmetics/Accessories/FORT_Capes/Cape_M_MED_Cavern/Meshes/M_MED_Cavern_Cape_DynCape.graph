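# Physics asset for the Cavern cape on the medium male body, simulating the dynamic cape bones against simplified body collision.
# Capsules cover pelvis, spine, clavicles and arms, with a flat box at the root.
# A center cape chain hangs from spine_05, and left and right chains both constrain to its last bone, dyn_cape04.
# Free-rotation constraints cross-link the side chains at segments 02, 04 and 05.
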PhysicsAsset /BRCosmetics/Accessories/FORT_Capes/Cape_M_MED_Cavern/Meshes/M_MED_Cavern_Cape_DynCape
bodies {
 clavicle_cape_l [{shapeType=Capsule, length=0.505, radius=0.505, rotation=<rot 0 -90 0>}]
 clavicle_cape_r [{shapeType=Capsule, length=0.505, radius=0.505, rotation=<rot 0 -90 0>}]
 clavicle_l [{shapeType=Capsule, length=7.1779, radius=10.5408, center=<11.7831 -1.586 -2.4859>, rotation=<rot 82.9629 10.6984 178.4662>}]
 clavicle_r [{shapeType=Capsule, length=6.4279, radius=11.2908, center=<-11.5831 0.5626 2.3092>, rotation=<rot 87.2048 8.3673 -3.8592>}]
 dyn_cape01 [{shapeType=Capsule, length=40, radius=6, center=<4 0 0>}]
 dyn_cape01_l [{shapeType=Capsule, length=15, radius=8, center=<4 -2 0>}]
 dyn_cape01_r [{shapeType=Capsule, length=15, radius=8, center=<4 -2 0>}]
 dyn_cape02 [{shapeType=Capsule, length=45, radius=6, center=<4.0879 -1.5 0>}]
 dyn_cape02_l [{shapeType=Capsule, length=30, radius=8, center=<4 -2 -10>}]
 dyn_cape02_r [{shapeType=Capsule, length=30, radius=8, center=<4 -2 10>}]
 dyn_cape03 [{shapeType=Capsule, length=45, radius=7, center=<4.0879 -2.3 0>}]
 dyn_cape03_l [{shapeType=Capsule, length=35, radius=10, center=<4 -2 -10>}]
 dyn_cape03_r [{shapeType=Capsule, length=35, radius=10, center=<4 -2 10>}]
 dyn_cape04 [{shapeType=Capsule, length=50, radius=8, center=<4.0879 0 0>}]
 dyn_cape04_l [{shapeType=Capsule, length=40, radius=11, center=<2 -1 -10>}]
 dyn_cape04_r [{shapeType=Capsule, length=40, radius=11, center=<2 -1 10>}]
 dyn_cape05_l [{shapeType=Capsule, length=40, radius=11, center=<4 -1 -10>}]
 dyn_cape05_r [{shapeType=Capsule, length=40, radius=11, center=<4 -1 10>}]
 lowerarm_l [{shapeType=Capsule, length=34.1377, radius=8.13, center=<18.227 0.9178 -2.1609>, rotation=<rot 82.7121 -0.0002 3.2014>}]
 lowerarm_r [{shapeType=Capsule, length=28.5439, radius=7.63, center=<-14.5234 -1.1325 0.8591>, rotation=<rot 82.7121 -0.0002 3.2014>}]
 pelvis [
  {shapeType=Capsule, length=12.6875, radius=19.0363, center=<-3.1042 0 -0.0004>}
  {shapeType=Capsule, length=17.6562, radius=16.7238, center=<-21.9244 0 0.0008>}
 ]
 root [{shapeType=Box, extentX=231.0625, extentY=231.0625, extentZ=12.375, center=<0 0 -4.1865>}]
 spine_01 [{shapeType=Capsule, length=12.875, radius=14.9738, center=<-3.2295 -3.663 0.0002>}]
 spine_02 [{shapeType=Capsule, length=15.125, radius=12.7238}]
 spine_03 [{shapeType=Capsule, length=16.125, radius=12.25}]
 spine_04 [{shapeType=Capsule, length=13.625, radius=14.75, center=<0.8861 -0.6413 -0.0001>}]
 spine_05 [{shapeType=Capsule, length=14.625, radius=13.5, center=<-0.9849 -0.7006 0.0001>}]
 upperarm_l [{shapeType=Capsule, length=20.4279, radius=9.5408, center=<13.2584 -0.8854 -1.0226>, rotation=<rot 87.3409 12.8748 -172.4072>}]
 upperarm_r [{shapeType=Capsule, length=19.6779, radius=9.5408, center=<-13.3097 0.8088 -0.2017>, rotation=<rot 87.3409 12.8742 9.1674>}]
}
constraints [
 {
  bone1Name=clavicle_cape_l
  bone2Name=clavicle_l
  swing1LimitDegrees=45
  swing1Motion=Free
  swing2LimitDegrees=45
  swing2Motion=Free
  twistLimitDegrees=45
  twistMotion=Free
 }
 {
  bone1Name=clavicle_cape_r
  bone2Name=clavicle_r
  swing1LimitDegrees=45
  swing1Motion=Free
  swing2LimitDegrees=45
  swing2Motion=Free
  twistLimitDegrees=45
  twistMotion=Free
 }
 {
  bone1Name=spine_05
  bone2Name=spine_04
  swing1LimitDegrees=45
  swing1Motion=Free
  swing2LimitDegrees=45
  swing2Motion=Free
  twistLimitDegrees=45
  twistMotion=Free
 }
 {
  bone1Name=spine_04
  bone2Name=spine_03
  swing1LimitDegrees=45
  swing1Motion=Free
  swing2LimitDegrees=45
  swing2Motion=Free
  twistLimitDegrees=45
  twistMotion=Free
 }
 {
  bone1Name=spine_03
  bone2Name=spine_02
  swing1LimitDegrees=45
  swing1Motion=Free
  swing2LimitDegrees=45
  swing2Motion=Free
  twistLimitDegrees=45
  twistMotion=Free
 }
 {
  bone1Name=spine_02
  bone2Name=spine_01
  swing1LimitDegrees=45
  swing1Motion=Free
  swing2LimitDegrees=45
  swing2Motion=Free
  twistLimitDegrees=45
  twistMotion=Free
 }
 {
  bone1Name=spine_01
  bone2Name=pelvis
  swing1LimitDegrees=45
  swing1Motion=Free
  swing2LimitDegrees=45
  swing2Motion=Free
  twistLimitDegrees=45
  twistMotion=Free
 }
 {
  bone1Name=lowerarm_l
  bone2Name=upperarm_l
  swing1LimitDegrees=45
  swing1Motion=Free
  swing2LimitDegrees=45
  swing2Motion=Free
  twistLimitDegrees=45
  twistMotion=Free
 }
 {
  bone1Name=lowerarm_r
  bone2Name=upperarm_r
  swing1LimitDegrees=45
  swing1Motion=Free
  swing2LimitDegrees=45
  swing2Motion=Free
  twistLimitDegrees=45
  twistMotion=Free
 }
 {
  bone1Name=pelvis
  bone2Name=root
  swing1LimitDegrees=45
  swing1Motion=Free
  swing2LimitDegrees=45
  swing2Motion=Free
  twistLimitDegrees=45
  twistMotion=Free
 }
 {
  bone1Name=upperarm_l
  bone2Name=clavicle_l
  swing1LimitDegrees=45
  swing1Motion=Free
  swing2LimitDegrees=45
  swing2Motion=Free
  twistLimitDegrees=45
  twistMotion=Free
 }
 {bone1Name=clavicle_r, bone2Name=spine_05, swing1LimitDegrees=45, swing2LimitDegrees=30, twistLimitDegrees=15}
 {
  bone1Name=clavicle_l
  bone2Name=spine_05
  swing1LimitDegrees=45
  swing1Motion=Free
  swing2LimitDegrees=45
  swing2Motion=Free
  twistLimitDegrees=45
  twistMotion=Free
 }
 {bone1Name=upperarm_r, bone2Name=clavicle_r, swing1LimitDegrees=45, swing2LimitDegrees=30, twistLimitDegrees=15}
 {bone1Name=dyn_cape01, bone2Name=spine_05, swing1LimitDegrees=45, swing2LimitDegrees=45, twistLimitDegrees=5, twistMotion=Locked}
 {bone1Name=dyn_cape02, bone2Name=dyn_cape01, swing1LimitDegrees=45, swing2LimitDegrees=45, twistLimitDegrees=0}
 {bone1Name=dyn_cape03, bone2Name=dyn_cape02, swing1LimitDegrees=45, swing2LimitDegrees=45, twistLimitDegrees=5}
 {bone1Name=dyn_cape04, bone2Name=dyn_cape03, swing1LimitDegrees=45, swing2LimitDegrees=45, twistLimitDegrees=5}
 {bone1Name=dyn_cape01_l, bone2Name=dyn_cape04, swing1LimitDegrees=30, swing2LimitDegrees=45, twistLimitDegrees=5}
 {bone1Name=dyn_cape02_l, bone2Name=dyn_cape01_l, swing1LimitDegrees=30, swing2LimitDegrees=45, twistLimitDegrees=5}
 {
  bone1Name=dyn_cape02_r
  bone2Name=dyn_cape02_l
  swing1LimitDegrees=45
  swing1Motion=Free
  swing2LimitDegrees=45
  swing2Motion=Free
  twistLimitDegrees=45
  twistMotion=Free
 }
 {bone1Name=dyn_cape03_l, bone2Name=dyn_cape02_l, swing1LimitDegrees=30, swing2LimitDegrees=45, twistLimitDegrees=5}
 {bone1Name=dyn_cape04_l, bone2Name=dyn_cape03_l, swing1LimitDegrees=30, swing2LimitDegrees=45, twistLimitDegrees=5}
 {
  bone1Name=dyn_cape04_r
  bone2Name=dyn_cape04_l
  swing1LimitDegrees=45
  swing1Motion=Free
  swing2LimitDegrees=45
  swing2Motion=Free
  twistLimitDegrees=45
  twistMotion=Free
 }
 {bone1Name=dyn_cape05_l, bone2Name=dyn_cape04_l, swing1LimitDegrees=20, swing2LimitDegrees=45, twistLimitDegrees=5}
 {
  bone1Name=dyn_cape05_r
  bone2Name=dyn_cape05_l
  swing1LimitDegrees=45
  swing1Motion=Free
  swing2LimitDegrees=45
  swing2Motion=Free
  twistLimitDegrees=45
  twistMotion=Free
 }
 {bone1Name=dyn_cape01_r, bone2Name=dyn_cape04, swing1LimitDegrees=30, swing2LimitDegrees=45, twistLimitDegrees=5}
 {bone1Name=dyn_cape02_r, bone2Name=dyn_cape01_r, swing1LimitDegrees=30, swing2LimitDegrees=45, twistLimitDegrees=5}
 {bone1Name=dyn_cape03_r, bone2Name=dyn_cape02_r, swing1LimitDegrees=30, swing2LimitDegrees=45, twistLimitDegrees=5}
 {bone1Name=dyn_cape04_r, bone2Name=dyn_cape03_r, swing1LimitDegrees=30, swing2LimitDegrees=45, twistLimitDegrees=5}
 {bone1Name=dyn_cape05_r, bone2Name=dyn_cape04_r, swing1LimitDegrees=20, swing2LimitDegrees=45, twistLimitDegrees=5}
]
